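# Sound cue for the move and rotate sound of a heist security camera, played in the ambient sound class.
# A random node picks one of six equally weighted camera rotation waves, which passes through a slight pitch modulator and a short fixed delay.
# It uses the camera movement attenuation asset and overrides concurrency to allow a few voices, stopping the quietest.
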SoundCue /HeistBasesGameplay/Audio/Turrets/Heist_Camera_MoveRotate_Cue
root=Delay_0
nodes [
 {id=Delay_0, children [Modulator_0], properties {delayMax=0.05, delayMin=0.05}}
 {id=Modulator_0, children [Random_0], properties {pitchMax=1.02, pitchMin=0.98}}
 {id=Random_0, children [WavePlayer_0,WavePlayer_2,WavePlayer_3,WavePlayer_6,WavePlayer_7,WavePlayer_8], properties {weights [1,1,1,1,1,1]}}
 {id=WavePlayer_0, properties {soundWaveAssetPtr=/Game/Athena/Sounds/Gadgets/SecurityTurret/Security_Camera_MoveRotate_01}}
 {id=WavePlayer_2, properties {soundWaveAssetPtr=/Game/Athena/Sounds/Gadgets/SecurityTurret/Security_Camera_MoveRotate_02}}
 {id=WavePlayer_3, properties {soundWaveAssetPtr=/Game/Athena/Sounds/Gadgets/SecurityTurret/Security_Camera_MoveRotate_03}}
 {id=WavePlayer_6, properties {soundWaveAssetPtr=/Game/Athena/Sounds/Gadgets/SecurityTurret/Security_Camera_MoveRotate_04}}
 {id=WavePlayer_7, properties {soundWaveAssetPtr=/Game/Athena/Sounds/Gadgets/SecurityTurret/Security_Camera_MoveRotate_05}}
 {id=WavePlayer_8, properties {soundWaveAssetPtr=/Game/Athena/Sounds/Gadgets/SecurityTurret/Security_Camera_MoveRotate_06}}
]
properties {
 attenuationSettings=/Game/Athena/Sounds/Gadgets/SecurityTurret/Security_Camera_Movement_Att
 bOverrideConcurrency=true
 maxDistance=2000
 soundClassObject=&Fort_Ambient_Athena
 volumeMultiplier=1.351
 attenuationOverrides {
  bEnableFocusInterpolation=true
  bEnableListenerFocus=true
  falloffDistance=1200
  nonFocusVolumeAttenuation=0.5
  stereoSpread=500
  attenuationShapeExtents {x=0}
 }
 concurrencyOverrides {maxCount=3, resolutionRule=StopQuietest}
}
