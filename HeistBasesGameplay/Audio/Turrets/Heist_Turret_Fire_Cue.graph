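# Sound cue for Heist turret gunfire, layering close, medium and distant shots, each with its own attenuation.
# The close layer randomizes turret samples with slight pitch and volume variation under the 3P fire sound class.
# Medium turret samples and borrowed HighInfantry rifle distant samples get fixed volume boosts and share the distant fire class.
# An owner-limited concurrency override stops the oldest sound, and the debug flag is on.
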SoundCue /HeistBasesGameplay/Audio/Turrets/Heist_Turret_Fire_Cue
root=Mixer_1
nodes [
 {id=Attenuation_118, children [Random_3], properties {attenuationSettings=&Fort_Weapon_Projectile_Close_Athena_Att}}
 {
  id=Attenuation_119
  children [Modulator_0]
  properties {attenuationSettings=/Game/Sounds/Attenuation/Fort_Weapon_Projectile_Medium_Sentry_Att}
 }
 {
  id=Attenuation_120
  children [Modulator_1]
  properties {attenuationSettings=/Game/Sounds/Attenuation/Fort_Weapon_Projectile_Distant_Sentry_Att}
 }
 {id=Mixer_1, children [Modulator_2, SoundClass_76], properties {inputVolume [1,0.6]}}
 {id=Mixer_4, children [Attenuation_119, Attenuation_120], properties {inputVolume [0.8,0.6]}}
 {id=Modulator_0, children [Random_0], properties {volumeMax=1.7, volumeMin=1.7}}
 {id=Modulator_1, children [Random_2], properties {volumeMax=1.8, volumeMin=1.8}}
 {id=Modulator_2, children [SoundClass_77], properties {pitchMax=1.1, pitchMin=0.9, volumeMax=1.1, volumeMin=0.9}}
 {id=Random_0, children [WavePlayer_0, WavePlayer_2, WavePlayer_3, WavePlayer_14], properties {weights [1,1,1,1]}}
 {id=Random_2, children [WavePlayer_7,WavePlayer_8,WavePlayer_9], properties {weights [1,1,1]}}
 {id=Random_3, children [WavePlayer_1, WavePlayer_10, WavePlayer_11, WavePlayer_12, WavePlayer_13], properties {weights [1,1,1,1,1]}}
 {id=SoundClass_76, children [Mixer_4], properties {soundClassOverride=&Fort_WeaponRanged_Fire_Distant}}
 {
  id=SoundClass_77
  children [Attenuation_118]
  properties {soundClassOverride=&Fort_WeaponRanged_Fire_3P_PSM}
 }
 {id=WavePlayer_0, properties {soundWaveAssetPtr=/Game/Athena/Sounds/Gadgets/SecurityTurret/Security_Turret_Fire_Medium_02}}
 {id=WavePlayer_1, properties {soundWaveAssetPtr=/Game/Athena/Sounds/Gadgets/SecurityTurret/Security_Turret_Fire_01}}
 {id=WavePlayer_10, properties {soundWaveAssetPtr=/Game/Athena/Sounds/Gadgets/SecurityTurret/Security_Turret_Fire_03}}
 {id=WavePlayer_11, properties {soundWaveAssetPtr=/Game/Athena/Sounds/Gadgets/SecurityTurret/Security_Turret_Fire_05}}
 {id=WavePlayer_12, properties {soundWaveAssetPtr=/Game/Athena/Sounds/Gadgets/SecurityTurret/Security_Turret_Fire_02}}
 {id=WavePlayer_13, properties {soundWaveAssetPtr=/Game/Athena/Sounds/Gadgets/SecurityTurret/Security_Turret_Fire_04}}
 {id=WavePlayer_14, properties {soundWaveAssetPtr=/Game/Athena/Sounds/Gadgets/SecurityTurret/Security_Turret_Fire_Medium_04}}
 {id=WavePlayer_2, properties {soundWaveAssetPtr=/Game/Athena/Sounds/Gadgets/SecurityTurret/Security_Turret_Fire_Medium_01}}
 {id=WavePlayer_3, properties {soundWaveAssetPtr=/Game/Athena/Sounds/Gadgets/SecurityTurret/Security_Turret_Fire_Medium_03}}
 {id=WavePlayer_7, properties {soundWaveAssetPtr=/Game/Sounds/Fort_Weapons/AutoRifle/HighInfantry/HiInfantry_Fire_Distant_01}}
 {id=WavePlayer_8, properties {soundWaveAssetPtr=/Game/Sounds/Fort_Weapons/AutoRifle/HighInfantry/HiInfantry_Fire_Distant_02}}
 {id=WavePlayer_9, properties {soundWaveAssetPtr=/Game/Sounds/Fort_Weapons/AutoRifle/HighInfantry/HiInfantry_Fire_Distant_03}}
]
properties {
 bDebug=true
 bOverrideConcurrency=true
 maxDistance=14000
 soundClassObject=&Fort_WeaponRanged_Fire_3P_PSM
 volumeMultiplier=0.35
 concurrencyOverrides {bLimitToOwner=true, maxCount=3, resolutionRule=StopOldest}
}
