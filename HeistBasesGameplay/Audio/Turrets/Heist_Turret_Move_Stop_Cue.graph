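# Sound cue for the sound a Heist security turret makes when it stops moving.
# A random node picks one of three equally weighted turret move-stop waves on each play.
# It uses the shared turret movement attenuation in the gadgets items sound class with a volume boost.
# An overridden concurrency limit caps it at three simultaneous instances, stopping the quietest.
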SoundCue /HeistBasesGameplay/Audio/Turrets/Heist_Turret_Move_Stop_Cue
root=Random_1
nodes [
 {id=Random_1, children [WavePlayer_1,WavePlayer_4,WavePlayer_5], properties {weights [1,1,1]}}
 {id=WavePlayer_1, properties {soundWaveAssetPtr=/Game/Athena/Sounds/Gadgets/SecurityTurret/Security_Turret_MoveStop_01}}
 {id=WavePlayer_4, properties {soundWaveAssetPtr=/Game/Athena/Sounds/Gadgets/SecurityTurret/Security_Turret_MoveStop_02}}
 {id=WavePlayer_5, properties {soundWaveAssetPtr=/Game/Athena/Sounds/Gadgets/SecurityTurret/Security_Turret_MoveStop_03}}
]
properties {
 attenuationSettings=/Game/Athena/Sounds/Gadgets/SecurityTurret/Security_Turret_Movement_Att
 bOverrideConcurrency=true
 maxDistance=2500
 soundClassObject=&Fort_Gadgets_Items
 volumeMultiplier=1.5
 attenuationOverrides {
  bEnableFocusInterpolation=true
  bEnableListenerFocus=true
  falloffDistance=1200
  nonFocusVolumeAttenuation=0.5
  stereoSpread=500
  attenuationShapeExtents {x=0}
 }
 concurrencyOverrides {maxCount=3, resolutionRule=StopQuietest}
}
