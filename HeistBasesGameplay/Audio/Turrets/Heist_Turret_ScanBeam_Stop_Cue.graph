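# Sound cue for a heist base security turret stopping its scan beam, played on the gadgets sound class with spherical turret attenuation.
# A mixer layers a fixed target-lost sound with one of three equally weighted, randomly chosen scan beam stop variations.
# A concurrency override limits it to one instance per owner, stopping the oldest when a new one starts.
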SoundCue /HeistBasesGameplay/Audio/Turrets/Heist_Turret_ScanBeam_Stop_Cue
root=Mixer_0
nodes [
 {id=Mixer_0, children [WavePlayer_0,Random_0], properties {inputVolume [1,1]}}
 {id=Random_0, children [WavePlayer_1,WavePlayer_2,WavePlayer_3], properties {weights [1,1,1]}}
 {id=WavePlayer_0, properties {soundWaveAssetPtr=/Game/Athena/Sounds/Gadgets/SecurityTurret/Security_Turret_Target_Lost_01}}
 {id=WavePlayer_1, properties {soundWaveAssetPtr=/Game/Athena/Sounds/Gadgets/SecurityTurret/Security_Turret_ScanBeam_Stop_01}}
 {id=WavePlayer_2, properties {soundWaveAssetPtr=/Game/Athena/Sounds/Gadgets/SecurityTurret/Security_Turret_ScanBeam_Stop_02}}
 {id=WavePlayer_3, properties {soundWaveAssetPtr=/Game/Athena/Sounds/Gadgets/SecurityTurret/Security_Turret_ScanBeam_Stop_03}}
]
properties {
 attenuationSettings=/Game/Athena/Sounds/Gadgets/SecurityTurret/Security_Turret_Sphere_Att
 bOverrideConcurrency=true
 maxDistance=3500
 soundClassObject=&Fort_Gadgets_Items
 volumeMultiplier=0.6
 concurrencyOverrides {bLimitToOwner=true, maxCount=1, resolutionRule=StopOldest}
}
