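# Sound cue for the targeting beam of a heist turret, playing a beam start sound and a beam loop together.
# A Mixer layers the two security turret waves, with the loop mixed slightly lower than the start.
# The cue routes to the gadgets and items sound class and references a wide sight-beam cone attenuation asset.
# A concurrency override caps simultaneous instances, stopping the quietest.
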SoundCue /HeistBasesGameplay/Audio/Turrets/Heist_Turret_TargetBeam_Loop_Cue
root=Mixer_1
nodes [
 {id=Mixer_1, children [WavePlayer_4,WavePlayer_0], properties {inputVolume [1,0.8]}}
 {id=WavePlayer_0, properties {soundWaveAssetPtr=/Game/Athena/Sounds/Gadgets/SecurityTurret/Security_Turret_TargetBeam_Loop_03}}
 {id=WavePlayer_4, properties {soundWaveAssetPtr=/Game/Athena/Sounds/Gadgets/SecurityTurret/Security_Turret_TargetBeam_Start_01}}
]
properties {
 attenuationSettings=/Game/Athena/Sounds/Gadgets/SecurityTurret/Security_Turret_Sight_Beam_Cone_Wide_Att
 bOverrideConcurrency=true
 maxDistance=2501
 soundClassObject=&Fort_Gadgets_Items
 volumeMultiplier=0.5
 attenuationOverrides {
  bEnableFocusInterpolation=true
  bEnableListenerFocus=true
  falloffDistance=1200
  nonFocusVolumeAttenuation=0.5
  stereoSpread=500
  attenuationShapeExtents {x=0}
 }
 concurrencyOverrides {maxCount=3, resolutionRule=StopQuietest}
}
